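# GPU Niagara system that spawns drifting, colored mesh petals over a flower field on the landscape.
# Importance sampling of a field mask texture places petals and a color texture is sampled alongside, followed by landscape sampling and culling outside a sphere or by matching a dark-green color.
# Curl noise, wind, gravity and drag move petals as they shrink and fade.
# The wind-gust Niagara parameter collection value scales spawn rate and wind strength.
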NiagaraSystem /FNGameplayCues/Effects/Fort_Effects/Effects/Environment/ColorFlowerField/NS_FlowerField_Color
dependencies {
 usedDataInterfaces [
  /Script/Niagara.NiagaraDataInterfaceTexture
  &NiagaraDataInterfaceLandscape
  &NiagaraDataInterfaceGBuffer
  &NiagaraDataInterfaceCurve
  &NiagaraDataInterfaceCollisionQuery
  &NiagaraDataInterfaceColorCurve
 ]
 usedDynamicInputs []
 usedModules [&SystemState]
}
emitters {FlowerFieldPetals {
  simTarget=GPUComputeSim
  emitterSpawnScript [/SetVariables(Emitter.Camera Position:NiagaraPosition=<54791 55693 3500>) {isSetParameters=true}]
  emitterUpdateScript [
   &EmitterState()
   &SpawnRate(
    SpawnRate=&Multiply_Float(A=3250, B=&Lerp_Float(A=0.25, B=2, Alpha=$NPC.FortniteNPC.WindGustsSystem_Active))
    Use Spawn Probability=false
   )
  ]
  particleSpawnScript [
   &InitializeParticle(
    Color=<rgba 1 1 1 0.75>
    Lifetime Max=8
    Lifetime Min=5
    Lifetime Mode=::NewEnumerator1
    Mesh Scale Mode=::NewEnumerator1
    Mesh Uniform Scale Max=1.5
    Mesh Uniform Scale Min=0.75
    Position=$Emitter.Camera Position
    Position Mode=::NewEnumerator1
    Sprite Size Mode=::NewEnumerator3
    Uniform Sprite Size=150
    UsePositionOffset=false
    Write Material Random=true
   )
   &ShapeLocation(MaskSpawnGroup=false, Override Seed=false, Sphere Radius=1500, UseOwnerScale=true) {enabled=false}
   /Niagara/Modules/Update/Textures/SampleTexture(
    EnableFalloff=false
    Generate UV Using Importance Sampling=true
    Importance Method:ENiagara_ImportanceColorMode=::NewEnumerator3
    Importance RGBA Channel=1
    Importance Texture=DataInterface() {propertyValues {
      Texture=/Game/Effects/Fort_Effects/Effects/Environment/ColorFlowerField/T_FlowerField
      TextureUserParameter {parameter {name=None, typeDefHandle {registeredTypeIndex=2}, varData []}}
     }}
    Kill Particles If Their Importance Sample Was Invalid=true
    Rotation Axis:Vector3f=<0 0 1>
    Rotation Coordinate Space:ENiagaraCoordinateSpace=::World
    Rotation Mode:ENiagara_RotationMode=::NewEnumerator1
    Sampled RGBA Mask:Vector4f=<1 1 1 1>
    Texture=DataInterface() {propertyValues {
      Texture=/Game/Effects/Fort_Effects/Effects/Environment/ColorFlowerField/T_FlowerField_Color
      TextureUserParameter {parameter {name=None, typeDefHandle {registeredTypeIndex=2}, varData []}}
     }}
    Transform UV Into Position=true
    UV Origin:NiagaraPosition=&AddVectorToPosition(Position=$Emitter.Camera Position, Vector=<0 0 0>)
    UV Scale:Vector2f=<30590 30590.5898>
   )
   &self:SampleLandscape() {module=SampleLandscape}
   &KillParticlesInVolume(Invert Volume=true, Sphere Radius=1500)
   &CameraOffset(Camera Offset Amount=-100)
   &AccelerationForce(Acceleration=<0 0 750>)
   &ApplyInitialForces()
   &self:Kill_By_GBuffer(HueMax=100, HueMin=60) {enabled=false, module=Kill_By_GBuffer}
   &self:Kill_By_Color_Comparison(Color To Kill:LinearColor=<rgba 0.0844 0.1384 0.0242 1>, Error Tolerance=0.1) {module=Kill_By_Color_Comparison}
  ]
  particleUpdateScript [
   &ParticleState()
   &ScaleMeshSize(Scale Factor=&VectorFromFloat(Value=&FloatFromCurve(FloatCurve=DataInterface() {propertyValues {Curve {keys [<at 0.5 1>,<at 1 0>]}}})))
   &UpdateMeshOrientation()
   &CurlNoiseForce(Bias Noise=false, Noise Frequency=5, Noise Strength=750, Pan Noise=false)
   &WindForce(
    Wind Speed=<400 0 10>
    Wind Speed Scale=&Lerp_Float(A=0.125, B=0.75, Alpha=$NPC.FortniteNPC.WindGustsSystem_Active)
   )
   &GravityForce(Gravity=<0 0 -90>)
   &Drag(Drag=10, Use Linear Drag=true, Use Rotational Drag=true)
   &SolveForcesAndVelocity(Clamp Velocity=false, Limit Acceleration=false)
   &DynamicMaterialParameters(
    Index 0 Param 4=&Lerp_Float(A=10, B=25, Alpha=&Sine(Normalized Angle=$Particles.Age, Period=1, Scale=1, Bias=0.5))
    Index 1 Param 1=1
    Module.Index0Param0Write=true
    Module.Index0Param1Write=true
    Module.Index0Param2Write=true
    Module.Index0Param3Write=true
    Module.Index1Param0Write=true
    Module.Index1Param1Write=true
    Module.Index1Param2Write=true
    Module.Index1Param3Write=true
    Write Parameter Index 1=true
   )
   &ScaleColor(
    Linear Color Curve=DataInterface() {propertyValues {
      AlphaCurve {keys [<at 0.8906 0.45>,<at 1 0>]}
      BlueCurve {keys [<at 0 1>,<at 1 1>]}
      GreenCurve {keys [<at 0 1>,<at 1 1>]}
      RedCurve {keys [<at 0 1>,<at 1 1>]}
     }}
    Scale Mode=::NewEnumerator2
   )
  ]
  rendererClasses [&NiagaraMeshRendererProperties]
  renderers [{rendererClass=&NiagaraMeshRendererProperties, rendererIndex=0}]
 }}
summary {
 systemName=NS_FlowerField_Color
 emitters [{emitterName=FlowerFieldPetals, simTarget=GPUComputeSim, rendererClasses [&NiagaraMeshRendererProperties]}]
 userVariables []
}
